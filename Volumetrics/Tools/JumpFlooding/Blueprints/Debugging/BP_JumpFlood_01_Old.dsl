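; Debug Blueprint actor that runs the Jump Flooding Algorithm on render targets to build distance field and gradient textures, with preview mesh and pass-counter text.
; Construction derives pass count as log2 of resolution; each pass ping-pongs between RT A and B through a dynamic material.
; Custom events run a seeded flood fully or step by step.
; Tick re-floods from a valid BP_VoxelizeMeshes render target, then writes the 3D distance and gradient targets.
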(event Custom|Find%20Edges%20%28Seed%29
  (Rendering|DrawMaterialtoRenderTarget (Variables|Default|GetRT%20A) (Variables|Default|GetSeed%20Mat))
  (Rendering|DrawMaterialtoRenderTarget (Variables|Default|GetRT%20B) (Variables|Default|GetSeed%20Mat))
  (Variables|Default|SetCompletedPasses 0)
  (Rendering|Material|SetScalarParameterValue (Variables|Default|GetJumpFloodMID) "Index" 0.0 :!class /Script/Engine.MaterialInstanceDynamic)
  (Rendering|Components|TextRender|SetText (Variables|Default|GetTextRender1) (Utilities|Text|MakeLiteralText "NSLOCTEXT(\"[77FF9324495FD96F8DCB0BAD45DB052A]\", \"FDC522394C4F95A3E9FB61881ADED91D\", \"Initial Seed\")")))

(event Custom|Jump%20Single%20Step
  (if (< (Variables|Default|GetCompletedPasses) (Variables|Default|GetRequired%20Passes))
    (CallFunction|SingleJumpStep)
    (Development|PrintString (Utilities|String|BuildString(Integer) (Utilities|String|BuildString(Integer) "Jump Flood Step " "" (Variables|Default|GetCompletedPasses) " of ") "" (Variables|Default|GetRequired%20Passes)))))

(event Custom|Jump%20Flood
  (CallFunction|FindEdges(Seed))
  (while (< (Variables|Default|GetCompletedPasses) (Variables|Default|GetRequired%20Passes))
    (CallFunction|SingleJumpStep))
  (Rendering|DrawMaterialtoRenderTarget "/Volumetrics/Tools/JumpFlooding/Textures/RT_JumpFlood_3D_4k_DF.RT_JumpFlood_3D_4k_DF" "/Volumetrics/Tools/JumpFlooding/Materials/M_JumpFlood_Display_3D.M_JumpFlood_Display_3D")
  (Rendering|ClearRenderTarget2D "/Volumetrics/Tools/JumpFlooding/Textures/RT_JumpFlood_3D_4k_Grad.RT_JumpFlood_3D_4k_Grad")
  (Rendering|DrawMaterialtoRenderTarget "/Volumetrics/Tools/JumpFlooding/Textures/RT_JumpFlood_3D_4k_Grad.RT_JumpFlood_3D_4k_Grad" "/Volumetrics/Tools/JumpFlooding/Materials/Generation/M_JumpFlood_Gradient_3D.M_JumpFlood_Gradient_3D"))

(event Custom|Jump%20Flood%20Seed%20RT
  (params (Seed%20RT :type Texture%20Render%20Target%202D%20Object%20Reference))
  (Variables|Default|SetCompletedPasses 0)
  (Rendering|ClearRenderTarget2D (Variables|Default|GetRT%20A))
  (Rendering|ClearRenderTarget2D (Variables|Default|GetRT%20B))
  (while (< (Variables|Default|GetCompletedPasses) (Variables|Default|GetRequired%20Passes))
    (CallFunction|SingleJumpStepwithSeed Seed%20RT))
  (Rendering|DrawMaterialtoRenderTarget "/Volumetrics/Tools/JumpFlooding/Textures/RT_JumpFlood_3D_4k_DF.RT_JumpFlood_3D_4k_DF" "/Volumetrics/Tools/JumpFlooding/Materials/M_JumpFlood_Display_3D.M_JumpFlood_Display_3D")
  (Rendering|ClearRenderTarget2D "/Volumetrics/Tools/JumpFlooding/Textures/RT_JumpFlood_3D_4k_Grad.RT_JumpFlood_3D_4k_Grad")
  (Rendering|DrawMaterialtoRenderTarget "/Volumetrics/Tools/JumpFlooding/Textures/RT_JumpFlood_3D_4k_Grad.RT_JumpFlood_3D_4k_Grad" "/Volumetrics/Tools/JumpFlooding/Materials/Generation/M_JumpFlood_Gradient_3D.M_JumpFlood_Gradient_3D"))

(event EventTick
  (params (DeltaSeconds :type Float%20%28single-precision%29))
  (Utilities|IsValid (Variables|Default|GetVoxelBP)
    (:Is%20Valid
      (CallFunction|JumpFloodSeedRT (Variables|/Volumetrics/Tools/VoxelizeMeshes/Blueprints/BP_VoxelizeMeshes.BP_VoxelizeMeshes_C|GetVoxelizationRTA (Variables|Default|GetVoxelBP))))))

(fn SingleJumpStep
  (bind _completedpasses (Variables|Default|GetCompletedPasses))
  (Rendering|Material|SetScalarParameterValue (Variables|Default|GetPreview%20RT%20MID) "Index" (Math|Conversions|ToFloat(Integer) (Math|Integer|%25%28Integer%29 (Variables|Default|GetCompletedPasses) 2)) :!class /Script/Engine.MaterialInstanceDynamic)
  (Rendering|Material|SetTextureParameterValue (Variables|Default|GetJumpFloodMID) "RT" (Utilities|Select :Index (Math|Integer|%25%28Integer%29 (Variables|Default|GetCompletedPasses) 2) :Option%200 (Variables|Default|GetRT%20A) :Option%201 (Variables|Default|GetRT%20B)))
  (Rendering|Material|SetScalarParameterValue (Variables|Default|GetJumpFloodMID) "Index" (Math|Conversions|ToFloat(Integer) (Variables|Default|GetCompletedPasses)) :!class /Script/Engine.MaterialInstanceDynamic)
  (Rendering|DrawMaterialtoRenderTarget (Utilities|Select :Index (Math|Integer|%25%28Integer%29 (+ (Variables|Default|GetCompletedPasses) 1) 2) :Option%200 (Variables|Default|GetRT%20A) :Option%201 (Variables|Default|GetRT%20B)) (Variables|Default|GetJumpFloodMID))
  (Math|Integer|IncrementInt _completedpasses
    (:%20%20
      (Rendering|Components|TextRender|SetText (Variables|Default|GetTextRender1) (Utilities|Text|ToText(Integer) _completedpasses))
      (Rendering|DrawMaterialtoRenderTarget "/Volumetrics/Tools/JumpFlooding/Textures/RT_JumpFlood_03_Result.RT_JumpFlood_03_Result" "/Volumetrics/Tools/JumpFlooding/Materials/M_JumpFlood_Display.M_JumpFlood_Display")
      (return))))

(fn SingleJumpStepWithSeed
  (params (Seed :type Texture%20Render%20Target%202D%20Object%20Reference))
  (bind _completedpasses (Variables|Default|GetCompletedPasses))
  (bind _returnvalue (+ _completedpasses 1))
  (bind _returnvalue_2 (Math|Integer|%25%28Integer%29 _returnvalue 2))
  (bind _rt_b (Variables|Default|GetRT%20B))
  (bind _rt_a (Variables|Default|GetRT%20A))
  (bind _returnvalue_3 (Utilities|Select :Index _returnvalue_2 :Option%200 _rt_a :Option%201 _rt_b))
  (bind _completedpasses_2 (Variables|Default|GetCompletedPasses))
  (Rendering|Material|SetTextureParameterValue (Variables|Default|GetJumpFloodMID) "RT" (select (== (Variables|Default|GetCompletedPasses) 0) Seed (Utilities|Select :Index (Math|Integer|%25%28Integer%29 (Variables|Default|GetCompletedPasses) 2) :Option%200 (Variables|Default|GetRT%20A) :Option%201 (Variables|Default|GetRT%20B))))
  (Rendering|Material|SetScalarParameterValue (Variables|Default|GetJumpFloodMID) "Index" (+ (Math|Conversions|ToFloat(Integer) (Variables|Default|GetCompletedPasses)) 1.0) :!class /Script/Engine.MaterialInstanceDynamic)
  (Rendering|ClearRenderTarget2D _returnvalue_3)
  (Rendering|DrawMaterialtoRenderTarget _returnvalue_3 (Variables|Default|GetJumpFloodMID))
  (Math|Integer|IncrementInt _completedpasses_2
    (:%20%20
      (Rendering|Components|TextRender|SetText (Variables|Default|GetTextRender1) (Utilities|Text|ToText(Integer) _completedpasses_2))
      (return))))

(fn ConstructionScript
  (bind (_x _y) (Math|Vector2D|BreakVector2D (Variables|Default|GetTextureSize)))
  (bind _rt_a (Variables|Default|GetRT%20A))
  (bind _sizex (Variables|/Script/Engine.TextureRenderTarget2D|GetSizeX _rt_a))
  (bind _sizey (Variables|/Script/Engine.TextureRenderTarget2D|GetSizeY _rt_a))
  (bind _returnvalue (Rendering|Material|CreateDynamicMaterialInstance (Variables|Default|GetJump%20Flood%20Material) "None" "None" :!class /Script/Engine.KismetMaterialLibrary))
  (Variables|Default|SetJumpFloodMID _returnvalue)
  (bind _returnvalue_2 (Rendering|Material|CreateDynamicMaterialInstance "/Volumetrics/Tools/JumpFlooding/Materials/Preview/M_PreviewRT.M_PreviewRT" "None" "None" :!class /Script/Engine.KismetMaterialLibrary))
  (Variables|Default|SetPreview%20RT%20MID _returnvalue_2)
  (if (not (Variables|Default|Get3D))
    (Variables|Default|SetResolution (Math|Integer|Max(Integer) _sizex _sizey)))
  (Variables|Default|SetTextureSize (Math|Vector2D|MakeVector2D (Math|Conversions|ToFloat(Integer) _sizex) (Math|Conversions|ToFloat(Integer) _sizey)))
  (Rendering|Material|SetVectorParameterValue (Variables|Default|GetJumpFloodMID) "TextureSize" (Utilities|Struct|MakeLinearColor _x _y) :!class /Script/Engine.MaterialInstanceDynamic)
  (Variables|Default|SetRequired%20Passes (Math|Float|Truncate (/ (Math|Float|Loge (Math|Conversions|ToFloat(Integer) (Variables|Default|GetResolution))) (Math|Float|Loge 2))))
  (Rendering|Material|SetMaterial (Variables|Default|GetStaticMesh) 0 (Variables|Default|GetPreview%20RT%20MID)))
