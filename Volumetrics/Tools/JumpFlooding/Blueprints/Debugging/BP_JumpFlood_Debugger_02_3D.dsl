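; Debug Blueprint that exercises JumpFloodComponent_3D and shows its render target results on a preview mesh.
; Numbered events initialize allocated or manually assigned render targets, run the full jump flood, step a single iteration, clear targets, detect edges and compute gradients.
; The construction script sets up a distance-field debug material instance with grid, distance and gradient toggles, whose RT texture is updated by the flood, iteration and edge-detection events.
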(event Custom|1%29%20Initialize
  (bind _jumpfloodcomponent_3d (Variables|Default|GetJumpFloodComponent_3D))
  (bind _jumpfloodcomponent_3d_2 (Variables|Default|GetJumpFloodComponent_3D))
  (bind _jumpfloodcomponent_3d_3 (Variables|Default|GetJumpFloodComponent_3D))
  (Variables|/Volumetrics/Tools/JumpFlooding/Blueprints/JumpFloodComponent_3D.JumpFloodComponent_3D_C|SetCurrentIndex 0 _jumpfloodcomponent_3d)
  (Variables|/Volumetrics/Tools/JumpFlooding/Blueprints/JumpFloodComponent_3D.JumpFloodComponent_3D_C|SetStartIndex 0 _jumpfloodcomponent_3d)
  ;[0.2867 0.6575 1] Initialize: Either Allocate or Assign Render Automatically or Targets Manually
  (if (Variables|Advanced|GetUsed%20Content%20Browser%20Render%20Targets)
    (Class|JumpFloodComponent3D|AssignRenderTargets _jumpfloodcomponent_3d_2 (Variables|Advanced|GetManual%20RT%20A) (Variables|Advanced|GetManual%20RT%20B) (Variables|Default|Get3D%20Size))
    (else
      (Class|JumpFloodComponent3D|AllocateRenderTargets _jumpfloodcomponent_3d_2 (Variables|Default|Get3D%20Size))
      ;[1 1 1] Assigning copy variables to see the component RTs. Not neceesary in real usage.
      (Variables|Debug|SetComponent%20RT%20A%20Reference (Variables|/Volumetrics/Tools/JumpFlooding/Blueprints/JumpFloodComponent_3D.JumpFloodComponent_3D_C|GetRT%20A _jumpfloodcomponent_3d_3))
      (Variables|Debug|SetComponent%20RT%20B%20Reference (Variables|/Volumetrics/Tools/JumpFlooding/Blueprints/JumpFloodComponent_3D.JumpFloodComponent_3D_C|GetRT%20B _jumpfloodcomponent_3d_3))
      (Rendering|Material|SetTextureParameterValue (Variables|MIDs|GetDebug%20RT%20MID) "RT" (Variables|Default|GetInput%20Texture)))))

(event Custom|2%29%20Jump%20Flood
  (CallFunction|1%29Initialize)
  ;[1 1 1] How to use the component when specifying a mask texture
  (bind _result_rt (Class|JumpFloodComponent3D|JumpFlood (Variables|Default|GetJumpFloodComponent_3D) (Variables|Default|GetSeed%20Mode) (Variables|Default|GetInput%20Texture) (Variables|Default|Get3D%20Size)))
  (Rendering|Material|SetTextureParameterValue (Variables|MIDs|GetDebug%20RT%20MID) "RT" _result_rt))

(event Custom|4%29Single%20Iteration
  ;[0.5052 0.004159 0] Debugs a Single Iteration. Not necessary for actual usage cases.
  (bind _result_rt (Class|JumpFloodComponent3D|SingleIteration (Variables|Default|GetJumpFloodComponent_3D)))
  (Rendering|Material|SetTextureParameterValue (Variables|MIDs|GetDebug%20RT%20MID) "RT" _result_rt))

(event Custom|5%29%20Clear%20RTs
  (bind _jumpfloodcomponent_3d (Variables|Default|GetJumpFloodComponent_3D))
  (bind _jumpfloodcomponent_3d_2 (Variables|Default|GetJumpFloodComponent_3D))
  ;[0.4896 0.004675 0] Debug Reset Button
  (Rendering|ClearRenderTarget2D (Variables|/Volumetrics/Tools/JumpFlooding/Blueprints/JumpFloodComponent_3D.JumpFloodComponent_3D_C|GetRT%20A _jumpfloodcomponent_3d_2))
  (Rendering|ClearRenderTarget2D (Variables|/Volumetrics/Tools/JumpFlooding/Blueprints/JumpFloodComponent_3D.JumpFloodComponent_3D_C|GetRT%20B _jumpfloodcomponent_3d_2))
  (Variables|/Volumetrics/Tools/JumpFlooding/Blueprints/JumpFloodComponent_3D.JumpFloodComponent_3D_C|SetCurrentIndex 0 _jumpfloodcomponent_3d)
  (Variables|/Volumetrics/Tools/JumpFlooding/Blueprints/JumpFloodComponent_3D.JumpFloodComponent_3D_C|SetStartIndex 0 _jumpfloodcomponent_3d))

(event Custom|6%29%20Detect%20Edges
  ;[1 0.6843 0.1604] Detect Edges
  (bind _result_rt (Class|JumpFloodComponent3D|DetectEdges (Variables|Default|GetJumpFloodComponent_3D) (Variables|Default|GetInput%20Texture) (Variables|Default|GetSeed%20Mode) (Variables|Default|Get3D%20Size)))
  ;[1 1 1] Debug the Edge Seed Result
  (bind _output_get (Variables|Debug|SetEdge%20Seeded%20Result _result_rt))
  (Rendering|Material|SetTextureParameterValue (Variables|MIDs|GetDebug%20RT%20MID) "RT" _output_get))

(event Custom|7%29%20Gradient
  (Class|JumpFloodComponent3D|ComputeGradient (Variables|Default|GetJumpFloodComponent_3D)))

(fn ConstructionScript
  (bind _debug_rt_mid (Variables|MIDs|GetDebug%20RT%20MID))
  (bind _debug_rt_mid_2 (Variables|MIDs|GetDebug%20RT%20MID))
  (Class|VolumetricsBPLIbrary|CreateAndAssignMID _debug_rt_mid "/Volumetrics/Tools/JumpFlooding/Materials/Debug/M_Debug_DistanceField_Texels.M_Debug_DistanceField_Texels")
  (Rendering|Material|SetMaterial (Variables|Default|GetPreview%20Mesh) 0 _debug_rt_mid)
  (Rendering|Material|SetScalarParameterValue _debug_rt_mid_2 "ShowGrid" (Math|Conversions|ToFloat(Boolean) (Variables|Default|GetShow%20Grid)) :!class /Script/Engine.MaterialInstanceDynamic)
  (Rendering|Material|SetScalarParameterValue _debug_rt_mid_2 "ShowDistance" (Math|Conversions|ToFloat(Boolean) (Variables|Default|GetShow%20Distance)) :!class /Script/Engine.MaterialInstanceDynamic)
  (Rendering|Material|SetScalarParameterValue _debug_rt_mid_2 "ShowGradient" (Math|Conversions|ToFloat(Boolean) (Variables|Default|GetShow%20Gradient)) :!class /Script/Engine.MaterialInstanceDynamic)
  (Rendering|Material|SetScalarParameterValue _debug_rt_mid_2 "DistanceDivisor" (Variables|Default|GetDistance%20Divisor) :!class /Script/Engine.MaterialInstanceDynamic))
